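; Ferret vehicle Blueprint, described as a plane in its comments, that disables empty seats when it is upside down near the ground.
; The IsSeatEnabled override wraps mesh roll and pitch toward 0–360 using loops with an iteration cap.
; If either angle is near 180 and the seat is empty, a stationary capsule trace below the Center socket checks for obstacles.
; A hit on a non-ignored actor class returns false; otherwise the parent's result applies.
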(event EventBeginPlay
  (|CallParentFunction) ; This node is disabled and will not be called.\nDrag off pins to build functionality.
)

(event Collision|EventActorBeginOverlap
  (params (OtherActor :type Actor%20Object%20Reference)))

(event EventTick
  (params (DeltaSeconds :type Float%20%28single-precision%29)))

(fn IsSeatEnabled
  (params (SeatIdx :type Integer))
  (bind _skeletalmesh (Variables|Vehicle|GetSkeletalMesh))
  (bind _returnvalue (Transformation|GetWorldRotation _skeletalmesh))
  (bind _cachedroll (Variables|GetCachedRoll))
  (bind _returnvalue_2 (< _cachedroll 0))
  (bind _cachedpitch (Variables|GetCachedPitch))
  (bind _returnvalue_3 (< _cachedpitch 0))
  (Utilities|FlowControl|Sequence
    (:then_0
      ;[1 1 1 0.5 14] Ensure proper pitch/roll values for calculations if the plane is upside down.
      ;; GetActualRotation
      (Variables|SetCachedRoll (.roll _returnvalue))
      (Variables|SetCachedPitch (.pitch _returnvalue))
      (Utilities|FlowControl|Sequence
        (:then_0
          (while (and (or (> _cachedroll 3.6e+02) _returnvalue_2) (<= (Variables|GetInfiniteBlocker) 5))
            (Variables|SetCachedRoll (+ (Variables|GetCachedRoll) (Math|Float|SelectFloat 3.6e+02 -3.6e+02 _returnvalue_2)))
            (Math|Integer|IncrementInt (Variables|GetInfiniteBlocker))))
        (:then_1
          (Variables|SetInfiniteBlocker 0))
        (:then_2
          (while (and (or (> _cachedpitch 3.6e+02) _returnvalue_3) (<= (Variables|GetInfiniteBlocker) 5))
            (Variables|SetCachedPitch (+ (Variables|GetCachedPitch) (Math|Float|SelectFloat 3.6e+02 -3.6e+02 _returnvalue_3)))
            (Math|Integer|IncrementInt (Variables|GetInfiniteBlocker))))))
    (:then_1
      ;[1 1 1 0.5] Disables empty seats if the plane is upsidedown near the ground.
      (if (or (and (> (LocalVariables|GetCachedRoll) 135.0) (< (LocalVariables|GetCachedRoll) 225.0)) (and (> (LocalVariables|GetCachedPitch) 135.0) (< (LocalVariables|GetCachedPitch) 225.0)))
        (Utilities|IsValid (Vehicle|GetPawnatSeat SeatIdx)
          (:Is%20Not%20Valid
            (bind _outhit (Collision|CapsuleTraceForObjects (- (Transformation|GetSocketLocation (Variables|Vehicle|GetSkeletalMesh) "Center") "0.000000,0.000000,100.000000") (- (Transformation|GetSocketLocation (Variables|Vehicle|GetSkeletalMesh) "Center") "0.000000,0.000000,100.000000") 1.5e+02 175.0 (Utilities|Array|MakeArray "ObjectTypeQuery1")))
            (Utilities|IsValid (.HitActor (Collision|BreakHitResult _outhit))
              (:Is%20Valid
                (if (not (Utilities|Set|ContainsItem (LocalVariables|GetActorTypesToIgnore) (Utilities|GetClass (.HitActor (Collision|BreakHitResult _outhit)))))
                  (return false))))))))
    (:then_2
      (bind _returnvalue_4 (|CallParentFunction SeatIdx))
      (return _returnvalue_4))))

(fn ConstructionScript
  (|CallParentFunction))
