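# Sound cue for the Ferret vehicle's destruction explosion in Figment, mixing a close layer and a distant layer.
# Each layer has its own sound class override and explosion attenuation, feeding a quality-level switch.
# The first quality slot randomizes two Ferret explosion waves, the second is empty, and the last two reuse one QuadRocket explosion wave.
# A concurrency override caps simultaneous instances.
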SoundCue /Figment_FerretVehicle/Blueprints/Sounds/Ferret_Destroy_Figment_Cue
root=Mixer_0
nodes [
 {id=Attenuation_0, children [QualityLevel_0], properties {attenuationSettings=/Game/Sounds/Attenuation/Fort_Explosion_Close_Athena_Att}}
 {id=Attenuation_1, children [QualityLevel_1], properties {attenuationSettings=/Game/Sounds/Attenuation/Fort_Explosion_Distant_Athena_Att}}
 {id=Mixer_0, children [SoundClass_0,SoundClass_1], properties {inputVolume [1,1]}}
 {id=QualityLevel_0, children [Random_2,null,WavePlayer_8,WavePlayer_8]}
 {id=QualityLevel_1, children [Random_1,null,WavePlayer_9,WavePlayer_9]}
 {id=Random_1, children [WavePlayer_6,WavePlayer_7], properties {weights [1,1]}}
 {id=Random_2, children [WavePlayer_0,WavePlayer_5], properties {weights [1,1]}}
 {id=SoundClass_0, children [Attenuation_0], properties {soundClassOverride=/Game/Sounds/FortSoundClassesAndModes/Fort_Explosives_Large_PSM}}
 {
  id=SoundClass_1
  children [Attenuation_1]
  properties {soundClassOverride=&Fort_WeaponRanged_Fire_Distant}
 }
 {id=WavePlayer_0, properties {soundWaveAssetPtr=&/Ferret_Weapon_Explode_Close_01}}
 {id=WavePlayer_5, properties {soundWaveAssetPtr=&/Ferret_Weapon_Explode_Close_02}}
 {id=WavePlayer_6, properties {soundWaveAssetPtr=&/Ferret_Weapon_Explode_Distant_01}}
 {id=WavePlayer_7, properties {soundWaveAssetPtr=&/Ferret_Weapon_Explode_Distant_02}}
 {id=WavePlayer_8, properties {soundWaveAssetPtr=&/QuadRocket_Explosion_Close_02}}
 {id=WavePlayer_9, properties {soundWaveAssetPtr=&/QuadRocket_Explosion_Distant_05}}
]
properties {bOverrideConcurrency=true, maxDistance=27500, volumeMultiplier=1, concurrencyOverrides {maxCount=3}}
